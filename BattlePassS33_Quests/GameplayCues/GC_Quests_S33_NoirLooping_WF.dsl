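; Looping gameplay cue that plays noir music for a quest while the local player is inside the target actor's noir day-sequence volume.
; On start it sizes an audio shape sphere slightly inside that child volume's radius.
; AudioGameplayVolume toggles attach music to the view target as a MusicEventSubsystem POI event, widening the sphere on entry and shrinking it on exit for hysteresis.
; Exit, cue removal and EndPlay fade out and unregister the music.
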(event Custom|ShrinkRadius
  (for _array_element (Actor|GetAllChildActors (stale EventOnLoopingStart(EitherVFX-preferthisversion) _asbp_state_driven_building_gameplay_actor_player_instance_quest_winterfest) true)
    (bind _asb_day_sequence_volume_noir_quest (Utilities|Casting|CastToB_DaySequenceVolume_NoirQuest _array_element))
    (Variables|Default|SetDaySequenceVolume _asb_day_sequence_volume_noir_quest)
    (Control|SetRadius (Variables|Default|GetAudioShapeSphere) (- (Variables|/Script/Engine.SphereComponent|GetSphereRadius (Variables|/Game/TimeOfDay/TODM_Sequence/Blueprint/Volume/B_DaySequenceVolume_NoirQuest.B_DaySequenceVolume_NoirQuest_C|GetSphere _asb_day_sequence_volume_noir_quest)) 7e+01) :!class /Script/AudioShapes.AudioShapeSphereComponent))
  (if (Variables|Default|GetPlayerInsideProxy)
    (CallFunction|ProxyCheckTrue?)))

(event EventOnLoopingStart(EitherVFX-preferthisversion)
  (params (MyTarget :type Actor%20Object%20Reference) (Parameters :type Gameplay%20Cue%20Parameters%20Structure%20%28by%20ref%29) (ParticleComponents :type Array%20of%20FXSystem%20Component%20Object%20References) (AudioComponents :type Array%20of%20Audio%20Component%20Object%20References))
  (|CallParentFunction MyTarget Parameters ParticleComponents AudioComponents)
  (Variables|Default|SetActive true)
  (bind _asbp_state_driven_building_gameplay_actor_player_instance_quest_winterfest (Utilities|Casting|CastToBP_StateDrivenBuildingGameplayActor_PlayerInstance_Quest_Winterfest MyTarget))
  (for _array_element (Actor|GetAllChildActors _asbp_state_driven_building_gameplay_actor_player_instance_quest_winterfest true)
    (bind _asb_day_sequence_volume_noir_quest (Utilities|Casting|CastToB_DaySequenceVolume_NoirQuest _array_element))
    (Variables|Default|SetDaySequenceVolume _asb_day_sequence_volume_noir_quest)
    (Control|SetRadius (Variables|Default|GetAudioShapeSphere) (- (Variables|/Script/Engine.SphereComponent|GetSphereRadius (Variables|/Game/TimeOfDay/TODM_Sequence/Blueprint/Volume/B_DaySequenceVolume_NoirQuest.B_DaySequenceVolume_NoirQuest_C|GetSphere _asb_day_sequence_volume_noir_quest)) 7e+01) :!class /Script/AudioShapes.AudioShapeSphereComponent))
  (if (Variables|Default|GetPlayerInsideProxy)
    (CallFunction|ProxyCheckTrue?)))

(event Custom|ProxyCheckTrue?
  (bind _self self)
  (bind _returnvalue (Fortnite|GetFirstLocalFortPlayerController _self true true))
  (bind _self_2 self)
  (bind _returnvalue_2 (GameInstanceSubsystems|GetMusicEventSubsystem _self_2))
  (Variables|Default|SetPlayerInsideProxy true)
  (if (Variables|Default|GetActive)
    (Variables|Default|GetNoirMusic
      (:then
        (CallFunction|FadeOutMusic))
      (:else
        _))
    (Utilities|IsValid _returnvalue
      (:Is%20Valid
        (bind _returnvalue_3 (Audio|SpawnSoundAttached (Variables|Default|GetMusicSoundCue) (Variables|/Script/Engine.Actor|GetRootComponent (Pawn|GetViewTarget _returnvalue))))
        (bind _output_get (Variables|Default|SetNoirMusic _returnvalue_3))
        (Audio|AddEventbyComponent _returnvalue_2 _output_get "(TagName=\"MusicEventSubsystem.Event.POI\")" "(GameplayTags=)" 0 "")
        (Audio|ChangeEventGroup _returnvalue_2 _output_get (Variables|Default|GetGroupName))
        (bind _daysequencevolume (Variables|Default|GetDaySequenceVolume)
          (:then
            (Control|SetRadius (Variables|Default|GetAudioShapeSphere) (+ (Variables|/Script/Engine.SphereComponent|GetSphereRadius (Variables|/Game/TimeOfDay/TODM_Sequence/Blueprint/Volume/B_DaySequenceVolume_NoirQuest.B_DaySequenceVolume_NoirQuest_C|GetSphere _daysequencevolume)) 7e+01) :!class /Script/AudioShapes.AudioShapeSphereComponent))
          (:else))))))

(event OnToggledOn(AudioGameplayVolume)
  (bind _self self)
  (bind _returnvalue (Fortnite|GetFirstLocalFortPlayerController _self true true))
  (bind _self_2 self)
  (bind _returnvalue_2 (GameInstanceSubsystems|GetMusicEventSubsystem _self_2))
  (Variables|Default|SetPlayerInsideProxy true)
  (if (Variables|Default|GetActive)
    (Variables|Default|GetNoirMusic
      (:then
        (CallFunction|FadeOutMusic))
      (:else
        _))
    (Utilities|IsValid _returnvalue
      (:Is%20Valid
        (bind _returnvalue_3 (Audio|SpawnSoundAttached (Variables|Default|GetMusicSoundCue) (Variables|/Script/Engine.Actor|GetRootComponent (Pawn|GetViewTarget _returnvalue))))
        (bind _output_get (Variables|Default|SetNoirMusic _returnvalue_3))
        (Audio|AddEventbyComponent _returnvalue_2 _output_get "(TagName=\"MusicEventSubsystem.Event.POI\")" "(GameplayTags=)" 0 "")
        (Audio|ChangeEventGroup _returnvalue_2 _output_get (Variables|Default|GetGroupName))
        (bind _daysequencevolume (Variables|Default|GetDaySequenceVolume)
          (:then
            (Control|SetRadius (Variables|Default|GetAudioShapeSphere) (+ (Variables|/Script/Engine.SphereComponent|GetSphereRadius (Variables|/Game/TimeOfDay/TODM_Sequence/Blueprint/Volume/B_DaySequenceVolume_NoirQuest.B_DaySequenceVolume_NoirQuest_C|GetSphere _daysequencevolume)) 7e+01) :!class /Script/AudioShapes.AudioShapeSphereComponent))
          (:else))))))

(event OnToggledOff(AudioGameplayVolume)
  (Variables|Default|SetPlayerInsideProxy false)
  (if (Variables|Default|GetActive)
    (CallFunction|FadeOutMusic)
    (CallFunction|ShrinkRadius)))

(event EventOnRemoval(EitherVFX-preferthisversion)
  (params (MyTarget :type Actor%20Object%20Reference) (Parameters :type Gameplay%20Cue%20Parameters%20Structure%20%28by%20ref%29) (ParticleComponents :type Array%20of%20FXSystem%20Component%20Object%20References) (AudioComponents :type Array%20of%20Audio%20Component%20Object%20References) (BurstCameraShakeInstance :type Camera%20Shake%20Base%20Object%20Reference) (BurstDecalInstance :type Decal%20Actor%20Object%20Reference))
  (|CallParentFunction MyTarget Parameters ParticleComponents AudioComponents BurstCameraShakeInstance BurstDecalInstance)
  (CallFunction|FadeOutMusic)
  (Variables|Default|SetActive false))

(event EventEndPlay
  (params (EndPlayReason :type EEndPlayReason%20Enum :default "Destroyed"))
  (CallFunction|FadeOutMusic)
  (Variables|Default|SetActive false))

(event Custom|FadeOutMusic
  (bind _noirmusic (Variables|Default|GetNoirMusic)
    (:then
      (Audio|Components|Audio|FadeOut _noirmusic 1.5 0.0 "Linear" :!class /Script/Engine.AudioComponent)
      (Audio|RemoveEvent (GameInstanceSubsystems|GetMusicEventSubsystem self) _noirmusic))
    (:else)))

(fn ConstructionScript)
